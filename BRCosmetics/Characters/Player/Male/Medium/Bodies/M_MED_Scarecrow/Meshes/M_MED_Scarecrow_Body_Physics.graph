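# Physics asset for the Scarecrow medium male player body, covering the legs and a dynamic jacket.
# Thighs and calves use capsules, while the jacket uses thin boxes on three two-bone chains per side, each hanging from a small master body.
# Each master links to its thigh with free angular motion, as calves do.
# Chain joints have tight twist and modest swing limits, narrower on one axis for the middle B chains.
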PhysicsAsset /BRCosmetics/Characters/Player/Male/Medium/Bodies/M_MED_Scarecrow/Meshes/M_MED_Scarecrow_Body_Physics
bodies {
 calf_l [{shapeType=Capsule, length=29.4915, radius=7.359, center=<-20.7496 1.1612 -0.4798>, rotation=<rot 0 90.7572 -89.0068>}]
 calf_r [{shapeType=Capsule, length=29.6175, radius=7.3051, center=<20.3886 -1.1242 -0.5872>, rotation=<rot 0 -89.5678 -87.3833>}]
 dyn_jacket_chainA_01_l [{shapeType=Box, extentX=20, extentY=2, extentZ=20, center=<10 0 0>, rotation=<rot 0 0 90>}]
 dyn_jacket_chainA_01_r [{shapeType=Box, extentX=20, extentY=2, extentZ=20, center=<10 0 0>, rotation=<rot 0 0 90>}]
 dyn_jacket_chainA_02_l [{shapeType=Box, extentX=20, extentY=2, extentZ=20, center=<10 0 0>, rotation=<rot 0 0 90>}]
 dyn_jacket_chainA_02_r [{shapeType=Box, extentX=20, extentY=2, extentZ=20, center=<10 0 0>, rotation=<rot 0 0 90>}]
 dyn_jacket_chainB_01_l [{shapeType=Box, extentX=20, extentY=2, extentZ=20, center=<10 0 0>, rotation=<rot 0 0 90>}]
 dyn_jacket_chainB_01_r [{shapeType=Box, extentX=20, extentY=2, extentZ=20, center=<10 0 0>, rotation=<rot 0 0 90>}]
 dyn_jacket_chainB_02_l [{shapeType=Box, extentX=20, extentY=2, extentZ=20, center=<10 0 0>, rotation=<rot 0 0 90>}]
 dyn_jacket_chainB_02_r [{shapeType=Box, extentX=20, extentY=2, extentZ=20, center=<10 0 0>, rotation=<rot 0 0 90>}]
 dyn_jacket_chainC_01_l [{shapeType=Box, extentX=20, extentY=2, extentZ=20, center=<10 0 0>, rotation=<rot 0 0 90>}]
 dyn_jacket_chainC_01_r [{shapeType=Box, extentX=20, extentY=2, extentZ=20, center=<10 0 0>, rotation=<rot 0 0 90>}]
 dyn_jacket_chainC_02_l [{shapeType=Box, extentX=20, extentY=2, extentZ=20, center=<10 0 0>, rotation=<rot 0 0 90>}]
 dyn_jacket_chainC_02_r [{shapeType=Box, extentX=20, extentY=2, extentZ=20, center=<10 0 0>, rotation=<rot 0 0 90>}]
 dyn_jacket_master_l [{shapeType=Box, extentX=1.01, extentY=1.01, extentZ=1.01, rotation=<rot 0 -90 0>}]
 dyn_jacket_master_r [{shapeType=Box, extentX=1.01, extentY=1.01, extentZ=1.01, rotation=<rot 0 -90 0>}]
 thigh_l [{shapeType=Capsule, length=35.522, radius=8.0766, center=<-20.443 0.0004 0>, rotation=<rot 90 0 0>}]
 thigh_r [{shapeType=Capsule, length=36.4584, radius=8.0766, center=<20 0 0>, rotation=<rot 90 0 0>}]
}
constraints [
 {
  bone1Name=calf_r
  bone2Name=thigh_r
  swing1LimitDegrees=45
  swing1Motion=Free
  swing2LimitDegrees=45
  swing2Motion=Free
  twistLimitDegrees=45
  twistMotion=Free
 }
 {
  bone1Name=calf_l
  bone2Name=thigh_l
  swing1LimitDegrees=45
  swing1Motion=Free
  swing2LimitDegrees=45
  swing2Motion=Free
  twistLimitDegrees=45
  twistMotion=Free
 }
 {bone1Name=dyn_jacket_chainA_01_l, bone2Name=dyn_jacket_master_l, swing1LimitDegrees=20, swing2LimitDegrees=20, twistLimitDegrees=5}
 {bone1Name=dyn_jacket_chainA_02_l, bone2Name=dyn_jacket_chainA_01_l, swing1LimitDegrees=20, swing2LimitDegrees=20, twistLimitDegrees=5}
 {bone1Name=dyn_jacket_chainB_01_l, bone2Name=dyn_jacket_master_l, swing1LimitDegrees=20, swing2LimitDegrees=10, twistLimitDegrees=5}
 {bone1Name=dyn_jacket_chainB_02_l, bone2Name=dyn_jacket_chainB_01_l, swing1LimitDegrees=20, swing2LimitDegrees=10, twistLimitDegrees=5}
 {bone1Name=dyn_jacket_chainC_01_l, bone2Name=dyn_jacket_master_l, swing1LimitDegrees=20, swing2LimitDegrees=20, twistLimitDegrees=5}
 {bone1Name=dyn_jacket_chainC_02_l, bone2Name=dyn_jacket_chainC_01_l, swing1LimitDegrees=20, swing2LimitDegrees=20, twistLimitDegrees=5}
 {bone1Name=dyn_jacket_chainA_01_r, bone2Name=dyn_jacket_master_r, swing1LimitDegrees=20, swing2LimitDegrees=20, twistLimitDegrees=5}
 {bone1Name=dyn_jacket_chainA_02_r, bone2Name=dyn_jacket_chainA_01_r, swing1LimitDegrees=20, swing2LimitDegrees=20, twistLimitDegrees=5}
 {bone1Name=dyn_jacket_chainB_01_r, bone2Name=dyn_jacket_master_r, swing1LimitDegrees=20, swing2LimitDegrees=10, twistLimitDegrees=5}
 {bone1Name=dyn_jacket_chainB_02_r, bone2Name=dyn_jacket_chainB_01_r, swing1LimitDegrees=20, swing2LimitDegrees=10, twistLimitDegrees=5}
 {bone1Name=dyn_jacket_chainC_01_r, bone2Name=dyn_jacket_master_r, swing1LimitDegrees=20, swing2LimitDegrees=20, twistLimitDegrees=5}
 {bone1Name=dyn_jacket_chainC_02_r, bone2Name=dyn_jacket_chainC_01_r, swing1LimitDegrees=20, swing2LimitDegrees=20, twistLimitDegrees=5}
 {
  bone1Name=dyn_jacket_master_l
  bone2Name=thigh_l
  swing1LimitDegrees=45
  swing1Motion=Free
  swing2LimitDegrees=45
  swing2Motion=Free
  twistLimitDegrees=45
  twistMotion=Free
 }
 {
  bone1Name=dyn_jacket_master_r
  bone2Name=thigh_r
  swing1LimitDegrees=45
  swing1Motion=Free
  swing2LimitDegrees=45
  swing2Motion=Free
  twistLimitDegrees=45
  twistMotion=Free
 }
]
