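# Post-process style AnimBP for the Scarecrow body cosmetic that adds secondary motion to back strings, necklace, jacket and scarf.
# The output chain copies the parent mesh pose, runs per-bone AnimDynamics, then a Flap node drives scarf offset from head rotation.
# Chain dynamics, constraint and RigidBody nodes sit outside that output chain.
# The update event checks once for a backpack mesh and, if found, sets a material parameter hiding an element.
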AnimBlueprint /BRCosmetics/Characters/Player/Male/Medium/Bodies/M_MED_Scarecrow/Meshes/M_MED_Scarecrow_Body_AnimBP
target_skeleton=&/M_MED_Scarecrow_Body_Skeleton
graphs [
 AnimGraph {nodes [
   n1 = "Anim Dynamics - Chain: dyn_jacket_chainA_01_l -> dyn_jacket_chainA_02_l"!(ComponentPose, ExternalForce="0,0,0") -> (Pose) {
    angularDampingOverride=0.999
    angularSpringConstant=100
    bAngularSpring=true
    bChain=true
    bGravityOverrideInSimSpace=true
    bOverrideAngularDamping=true
    numSolverIterationsPostUpdate=4
    numSolverIterationsPreUpdate=16
    simulationSpace=RootRelative
    alphaBoolBlend {blendOption=Linear}
    boundBone {boneName=dyn_jacket_chainA_01_l}
    chainEnd {boneName=dyn_jacket_chainA_02_l}
    componentLinearAccScale=<1 1 1>
    componentLinearVelScale=<1 1 1>
    physicsBodyDefinitions [{
      sphereCollisionRadius=0
      boundBone {boneName=dyn_jacket_chainA_01_l}
      boxExtents=<12 20 25>
      constraintSetup {angularLimitsMax=<0 30 10>, angularLimitsMin=<0 0 -10>, angularTarget=<0 0 1>}
      localJointOffset=<3.5 0 0>
     }]
   }
   n2 = "Anim Dynamics - Chain: dyn_jacket_chainB_01_l -> dyn_jacket_chainB_02_l"!(ComponentPose=@n1, ExternalForce="0,0,0") -> (Pose) {
    angularDampingOverride=0.999
    angularSpringConstant=100
    bAngularSpring=true
    bChain=true
    bGravityOverrideInSimSpace=true
    bOverrideAngularDamping=true
    numSolverIterationsPostUpdate=4
    numSolverIterationsPreUpdate=16
    simulationSpace=RootRelative
    alphaBoolBlend {blendOption=Linear}
    boundBone {boneName=dyn_jacket_chainB_01_l}
    chainEnd {boneName=dyn_jacket_chainB_02_l}
    componentLinearAccScale=<1 1 1>
    componentLinearVelScale=<1 1 1>
    physicsBodyDefinitions [{
      sphereCollisionRadius=0
      boundBone {boneName=dyn_jacket_chainB_01_l}
      boxExtents=<12 20 25>
      constraintSetup {angularLimitsMax=<0 30 10>, angularLimitsMin=<0 0 -10>, angularTarget=<0 0 1>}
      localJointOffset=<3.5 0 0>
     }]
   }
   n3 = "Anim Dynamics - Chain: dyn_jacket_chainC_01_l -> dyn_jacket_chainC_02_l"!(ComponentPose=@n2, ExternalForce="0,0,0") -> (Pose) {
    angularDampingOverride=0.999
    angularSpringConstant=100
    bAngularSpring=true
    bChain=true
    bGravityOverrideInSimSpace=true
    bOverrideAngularDamping=true
    numSolverIterationsPostUpdate=4
    numSolverIterationsPreUpdate=16
    simulationSpace=RootRelative
    alphaBoolBlend {blendOption=Linear}
    boundBone {boneName=dyn_jacket_chainC_01_l}
    chainEnd {boneName=dyn_jacket_chainC_02_l}
    componentLinearAccScale=<1 1 1>
    componentLinearVelScale=<1 1 1>
    physicsBodyDefinitions [{
      sphereCollisionRadius=0
      boundBone {boneName=dyn_jacket_chainC_01_l}
      boxExtents=<12 20 25>
      constraintSetup {angularLimitsMax=<0 30 10>, angularLimitsMin=<0 0 -10>, angularTarget=<0 0 1>}
      localJointOffset=<3.5 0 0>
     }]
   }
   n4 = "Anim Dynamics - Chain: dyn_jacket_chainA_01_r -> dyn_jacket_chainA_02_r"!(ComponentPose=@n3, ExternalForce="0,0,0") -> (Pose) {
    angularDampingOverride=0.999
    angularSpringConstant=100
    bAngularSpring=true
    bChain=true
    bGravityOverrideInSimSpace=true
    bOverrideAngularDamping=true
    numSolverIterationsPostUpdate=4
    numSolverIterationsPreUpdate=16
    simulationSpace=RootRelative
    alphaBoolBlend {blendOption=Linear}
    boundBone {boneName=dyn_jacket_chainA_01_r}
    chainEnd {boneName=dyn_jacket_chainA_02_r}
    componentLinearAccScale=<1 1 1>
    componentLinearVelScale=<1 1 1>
    physicsBodyDefinitions [{
      sphereCollisionRadius=0
      boundBone {boneName=dyn_jacket_chainA_01_r}
      boxExtents=<12 20 25>
      constraintSetup {angularLimitsMax=<0 30 10>, angularLimitsMin=<0 0 -10>, angularTarget=<0 0 1>}
      localJointOffset=<3.5 0 0>
     }]
   }
   n5 = "Anim Dynamics - Chain: dyn_jacket_chainB_01_r -> dyn_jacket_chainB_02_r"!(ComponentPose=@n4, ExternalForce="0,0,0") -> (Pose) {
    angularDampingOverride=0.999
    angularSpringConstant=100
    bAngularSpring=true
    bChain=true
    bGravityOverrideInSimSpace=true
    bOverrideAngularDamping=true
    numSolverIterationsPostUpdate=4
    numSolverIterationsPreUpdate=16
    simulationSpace=RootRelative
    alphaBoolBlend {blendOption=Linear}
    boundBone {boneName=dyn_jacket_chainB_01_r}
    chainEnd {boneName=dyn_jacket_chainB_02_r}
    componentLinearAccScale=<1 1 1>
    componentLinearVelScale=<1 1 1>
    physicsBodyDefinitions [{
      sphereCollisionRadius=0
      boundBone {boneName=dyn_jacket_chainB_01_r}
      boxExtents=<12 20 25>
      constraintSetup {angularLimitsMax=<0 30 10>, angularLimitsMin=<0 0 -10>, angularTarget=<0 0 1>}
      localJointOffset=<3.5 0 0>
     }]
   }
   "Anim Dynamics - Chain: dyn_jacket_chainC_01_r -> dyn_jacket_chainC_02_r"!(ComponentPose=@n5, ExternalForce="0,0,0") -> (Pose) {
    angularDampingOverride=0.999
    angularSpringConstant=100
    bAngularSpring=true
    bChain=true
    bGravityOverrideInSimSpace=true
    bOverrideAngularDamping=true
    numSolverIterationsPostUpdate=4
    numSolverIterationsPreUpdate=16
    simulationSpace=RootRelative
    alphaBoolBlend {blendOption=Linear}
    boundBone {boneName=dyn_jacket_chainC_01_r}
    chainEnd {boneName=dyn_jacket_chainC_02_r}
    componentLinearAccScale=<1 1 1>
    componentLinearVelScale=<1 1 1>
    physicsBodyDefinitions [{
      sphereCollisionRadius=0
      boundBone {boneName=dyn_jacket_chainC_01_r}
      boxExtents=<12 20 25>
      constraintSetup {angularLimitsMax=<0 30 10>, angularLimitsMin=<0 0 -10>, angularTarget=<0 0 1>}
      localJointOffset=<3.5 0 0>
     }]
   }
   Anim Dynamics - Bone: dyn_back_strings(Alpha=1, ExternalForce="0,0,0") {
    angularDampingOverride=0.99
    angularSpringConstant=500
    bAngularSpring=true
    bGravityOverrideInSimSpace=true
    bOverrideAngularDamping=true
    gravityScale=0.3
    alphaBoolBlend {blendOption=Linear}
    boundBone {boneName=dyn_back_strings}
    componentLinearAccScale=<1 1 1>
    componentLinearVelScale=<1 1 1>
    gravityOverride {z=-200}
    physicsBodyDefinitions [{
      sphereCollisionRadius=0
      boundBone {boneName=dyn_back_strings}
      boxExtents=<20 50 50>
      constraintSetup {angularLimitsMax=<0 40 10>, angularLimitsMin=<0 -40 -120>, angularTarget=<0 0 0>}
      localJointOffset=<20 0 0>
     }]
   }
   Blend Poses by bool(BlendPose_0=Use cached pose 'Cache_01'(), bActiveValue=Get bClothEnabled()) {blendType=Linear}
   Component To Local()
   "Transform (Modify) Bone"(Alpha=1, Translation="0,0,0", Rotation="0,0,0", Scale="1,1,1", None=Get bIsTargeting()) {alphaBoolBlend {blendOption=Linear}, boneToModify=~}
   n6 = "Constraint - dyn_jacket_master_l (P)"(Alpha=1, ConstraintWeights_0=1) {
    alphaBoolBlend {blendOption=Linear}
    boneToModify {boneName=dyn_jacket_master_l}
    constraintSetup [{targetBone {boneName=pelvis}}]
    constraintWeights [1]
   }
   n7 = "Constraint - dyn_jacket_master_r (P)"(ComponentPose=@n6, Alpha=1, ConstraintWeights_0=1) {
    alphaBoolBlend {blendOption=Linear}
    boneToModify {boneName=dyn_jacket_master_r}
    constraintSetup [{targetBone {boneName=pelvis}}]
    constraintWeights [1]
   }
   RigidBody(ComponentPose=@n7) {
    bOverrideWorldGravity=true
    overridePhysicsAsset=&/M_MED_Scarecrow_Body_Physics
    simulationSpace=BaseBoneSpace
    alphaBoolBlend {blendOption=Linear}
    baseBoneRef {boneName=root}
    overrideWorldGravity {z=-980}
   }
   n8 = Anim Dynamics - Bone: dyn_back_strings(ComponentPose=Local To Component(LocalPose=Copy Pose From Mesh() {bUseAttachedParent=true}), Alpha=0.5, ExternalForce="0,0,0") {
    angularDampingOverride=0.99
    angularSpringConstant=600
    bAngularSpring=true
    bGravityOverrideInSimSpace=true
    bOverrideAngularDamping=true
    bUseGravityOverride=true
    gravityScale=0.01
    alphaBoolBlend {blendOption=Linear}
    boundBone {boneName=dyn_back_strings}
    componentLinearAccScale=<0.5 0.5 0.5>
    componentLinearVelScale=<0.5 0.5 0.5>
    gravityOverride {z=-500}
    physicsBodyDefinitions [{
      sphereCollisionRadius=0
      boundBone {boneName=dyn_back_strings}
      boxExtents=<5 15 15>
      constraintSetup {angularLimitsMax=<0 30 0>, angularLimitsMin=<0 -30 -120>, angularTarget=<0 0 0>}
      localJointOffset=<5 0 0>
     }]
   }
   n9 = Anim Dynamics - Bone: dyn_necklace(ComponentPose=@n8, Alpha=1, ExternalForce="0,0,0") {
    angularDampingOverride=0.98
    bAngularSpring=true
    bGravityOverrideInSimSpace=true
    bOverrideAngularDamping=true
    bUseGravityOverride=true
    alphaBoolBlend {blendOption=Linear}
    boundBone {boneName=dyn_necklace}
    componentLinearAccScale=<1 1 1>
    componentLinearVelScale=<1 1 1>
    gravityOverride {z=-980}
    physicsBodyDefinitions [{
      sphereCollisionRadius=0
      boundBone {boneName=dyn_necklace}
      boxExtents=<5 20 20>
      constraintSetup {angularLimitsMax=<0 0 20>, angularLimitsMin=<0 -40 -20>, angularTarget=<0 0 0>}
      localJointOffset=<5 0 0>
     }]
   }
   n10 = Anim Dynamics - Bone: dyn_jacket_chainA_01_l!(ComponentPose=@n9, ExternalForce="0,0,0") -> (Pose) {
    angularDampingOverride=0.999
    angularSpringConstant=100
    bAngularSpring=true
    bGravityOverrideInSimSpace=true
    bOverrideAngularDamping=true
    numSolverIterationsPostUpdate=4
    numSolverIterationsPreUpdate=16
    simulationSpace=RootRelative
    alphaBoolBlend {blendOption=Linear}
    boundBone {boneName=dyn_jacket_chainA_01_l}
    chainEnd {boneName=dyn_jacket_chainA_02_l}
    componentLinearAccScale=<1 1 1>
    componentLinearVelScale=<1 1 1>
    physicsBodyDefinitions [{
      sphereCollisionRadius=0
      boundBone {boneName=dyn_jacket_chainA_01_l}
      boxExtents=<12 20 25>
      constraintSetup {angularLimitsMax=<0 30 10>, angularLimitsMin=<0 0 -10>, angularTarget=<0 0 1>}
      localJointOffset=<3.5 0 0>
     }]
   }
   n11 = Anim Dynamics - Bone: dyn_jacket_chainB_01_l!(ComponentPose=@n10, ExternalForce="0,0,0") -> (Pose) {
    angularDampingOverride=0.999
    angularSpringConstant=100
    bAngularSpring=true
    bGravityOverrideInSimSpace=true
    bOverrideAngularDamping=true
    numSolverIterationsPostUpdate=4
    numSolverIterationsPreUpdate=16
    simulationSpace=RootRelative
    alphaBoolBlend {blendOption=Linear}
    boundBone {boneName=dyn_jacket_chainB_01_l}
    chainEnd {boneName=dyn_jacket_chainB_02_l}
    componentLinearAccScale=<1 1 1>
    componentLinearVelScale=<1 1 1>
    physicsBodyDefinitions [{
      sphereCollisionRadius=0
      boundBone {boneName=dyn_jacket_chainB_01_l}
      boxExtents=<12 20 25>
      constraintSetup {angularLimitsMax=<0 30 10>, angularLimitsMin=<0 0 -10>, angularTarget=<0 0 1>}
      localJointOffset=<3.5 0 0>
     }]
   }
   n12 = Anim Dynamics - Bone: dyn_jacket_chainC_01_l!(ComponentPose=@n11, ExternalForce="0,0,0") -> (Pose) {
    angularDampingOverride=0.999
    angularSpringConstant=100
    bAngularSpring=true
    bGravityOverrideInSimSpace=true
    bOverrideAngularDamping=true
    numSolverIterationsPostUpdate=4
    numSolverIterationsPreUpdate=16
    simulationSpace=RootRelative
    alphaBoolBlend {blendOption=Linear}
    boundBone {boneName=dyn_jacket_chainC_01_l}
    chainEnd {boneName=dyn_jacket_chainC_02_l}
    componentLinearAccScale=<1 1 1>
    componentLinearVelScale=<1 1 1>
    physicsBodyDefinitions [{
      sphereCollisionRadius=0
      boundBone {boneName=dyn_jacket_chainC_01_l}
      boxExtents=<12 20 25>
      constraintSetup {angularLimitsMax=<0 30 10>, angularLimitsMin=<0 0 -10>, angularTarget=<0 0 1>}
      localJointOffset=<3.5 0 0>
     }]
   }
   n13 = Anim Dynamics - Bone: dyn_jacket_chainA_01_r!(ComponentPose=@n12, ExternalForce="0,0,0") -> (Pose) {
    angularDampingOverride=0.999
    angularSpringConstant=100
    bAngularSpring=true
    bGravityOverrideInSimSpace=true
    bOverrideAngularDamping=true
    numSolverIterationsPostUpdate=4
    numSolverIterationsPreUpdate=16
    simulationSpace=RootRelative
    alphaBoolBlend {blendOption=Linear}
    boundBone {boneName=dyn_jacket_chainA_01_r}
    chainEnd {boneName=dyn_jacket_chainA_02_r}
    componentLinearAccScale=<1 1 1>
    componentLinearVelScale=<1 1 1>
    physicsBodyDefinitions [{
      sphereCollisionRadius=0
      boundBone {boneName=dyn_jacket_chainA_01_r}
      boxExtents=<12 20 25>
      constraintSetup {angularLimitsMax=<0 30 10>, angularLimitsMin=<0 0 -10>, angularTarget=<0 0 1>}
      localJointOffset=<3.5 0 0>
     }]
   }
   n14 = Anim Dynamics - Bone: dyn_jacket_chainB_01_r!(ComponentPose=@n13, ExternalForce="0,0,0") -> (Pose) {
    angularDampingOverride=0.999
    angularSpringConstant=100
    bAngularSpring=true
    bGravityOverrideInSimSpace=true
    bOverrideAngularDamping=true
    numSolverIterationsPostUpdate=4
    numSolverIterationsPreUpdate=16
    simulationSpace=RootRelative
    alphaBoolBlend {blendOption=Linear}
    boundBone {boneName=dyn_jacket_chainB_01_r}
    chainEnd {boneName=dyn_jacket_chainB_02_r}
    componentLinearAccScale=<1 1 1>
    componentLinearVelScale=<1 1 1>
    physicsBodyDefinitions [{
      sphereCollisionRadius=0
      boundBone {boneName=dyn_jacket_chainB_01_r}
      boxExtents=<12 20 25>
      constraintSetup {angularLimitsMax=<0 30 10>, angularLimitsMin=<0 0 -10>, angularTarget=<0 0 1>}
      localJointOffset=<3.5 0 0>
     }]
   }
   n15 = Anim Dynamics - Bone: dyn_jacket_chainC_01_r!(ComponentPose=@n14, ExternalForce="0,0,0") -> (Pose) {
    angularDampingOverride=0.999
    angularSpringConstant=100
    bAngularSpring=true
    bGravityOverrideInSimSpace=true
    bOverrideAngularDamping=true
    numSolverIterationsPostUpdate=4
    numSolverIterationsPreUpdate=16
    simulationSpace=RootRelative
    alphaBoolBlend {blendOption=Linear}
    boundBone {boneName=dyn_jacket_chainC_01_r}
    chainEnd {boneName=dyn_jacket_chainC_02_r}
    componentLinearAccScale=<1 1 1>
    componentLinearVelScale=<1 1 1>
    physicsBodyDefinitions [{
      sphereCollisionRadius=0
      boundBone {boneName=dyn_jacket_chainC_01_r}
      boxExtents=<12 20 25>
      constraintSetup {angularLimitsMax=<0 30 10>, angularLimitsMin=<0 0 -10>, angularTarget=<0 0 1>}
      localJointOffset=<3.5 0 0>
     }]
   }
   n16 = Anim Dynamics - Bone: dyn_jacket_chainA_02_l!(ComponentPose=@n15, ExternalForce="0,0,0") -> (Pose) {
    angularDampingOverride=0.999
    angularSpringConstant=100
    bAngularSpring=true
    bGravityOverrideInSimSpace=true
    bOverrideAngularDamping=true
    numSolverIterationsPostUpdate=4
    numSolverIterationsPreUpdate=16
    simulationSpace=RootRelative
    alphaBoolBlend {blendOption=Linear}
    boundBone {boneName=dyn_jacket_chainA_02_l}
    chainEnd {boneName=dyn_jacket_chainA_02_l}
    componentLinearAccScale=<1 1 1>
    componentLinearVelScale=<1 1 1>
    physicsBodyDefinitions [{
      sphereCollisionRadius=0
      boundBone {boneName=dyn_jacket_chainA_02_l}
      boxExtents=<12 20 25>
      constraintSetup {angularLimitsMax=<0 30 10>, angularLimitsMin=<0 0 -10>, angularTarget=<0 0 1>}
      localJointOffset=<3.5 0 0>
     }]
   }
   n17 = Anim Dynamics - Bone: dyn_jacket_chainB_02_l!(ComponentPose=@n16, ExternalForce="0,0,0") -> (Pose) {
    angularDampingOverride=0.999
    angularSpringConstant=100
    bAngularSpring=true
    bGravityOverrideInSimSpace=true
    bOverrideAngularDamping=true
    numSolverIterationsPostUpdate=4
    numSolverIterationsPreUpdate=16
    simulationSpace=RootRelative
    alphaBoolBlend {blendOption=Linear}
    boundBone {boneName=dyn_jacket_chainB_02_l}
    chainEnd {boneName=dyn_jacket_chainB_02_l}
    componentLinearAccScale=<1 1 1>
    componentLinearVelScale=<1 1 1>
    physicsBodyDefinitions [{
      sphereCollisionRadius=0
      boundBone {boneName=dyn_jacket_chainB_02_l}
      boxExtents=<12 20 25>
      constraintSetup {angularLimitsMax=<0 30 10>, angularLimitsMin=<0 0 -10>, angularTarget=<0 0 1>}
      localJointOffset=<3.5 0 0>
     }]
   }
   n18 = Anim Dynamics - Bone: dyn_jacket_chainC_02_l!(ComponentPose=@n17, ExternalForce="0,0,0") -> (Pose) {
    angularDampingOverride=0.999
    angularSpringConstant=100
    bAngularSpring=true
    bGravityOverrideInSimSpace=true
    bOverrideAngularDamping=true
    numSolverIterationsPostUpdate=4
    numSolverIterationsPreUpdate=16
    simulationSpace=RootRelative
    alphaBoolBlend {blendOption=Linear}
    boundBone {boneName=dyn_jacket_chainC_02_l}
    chainEnd {boneName=dyn_jacket_chainC_02_l}
    componentLinearAccScale=<1 1 1>
    componentLinearVelScale=<1 1 1>
    physicsBodyDefinitions [{
      sphereCollisionRadius=0
      boundBone {boneName=dyn_jacket_chainC_02_l}
      boxExtents=<12 20 25>
      constraintSetup {angularLimitsMax=<0 30 10>, angularLimitsMin=<0 0 -10>, angularTarget=<0 0 1>}
      localJointOffset=<3.5 0 0>
     }]
   }
   n19 = Anim Dynamics - Bone: dyn_jacket_chainA_02_r!(ComponentPose=@n18, ExternalForce="0,0,0") -> (Pose) {
    angularDampingOverride=0.999
    angularSpringConstant=100
    bAngularSpring=true
    bGravityOverrideInSimSpace=true
    bOverrideAngularDamping=true
    numSolverIterationsPostUpdate=4
    numSolverIterationsPreUpdate=16
    simulationSpace=RootRelative
    alphaBoolBlend {blendOption=Linear}
    boundBone {boneName=dyn_jacket_chainA_02_r}
    chainEnd {boneName=dyn_jacket_chainA_02_r}
    componentLinearAccScale=<1 1 1>
    componentLinearVelScale=<1 1 1>
    physicsBodyDefinitions [{
      sphereCollisionRadius=0
      boundBone {boneName=dyn_jacket_chainA_02_r}
      boxExtents=<12 20 25>
      constraintSetup {angularLimitsMax=<0 30 10>, angularLimitsMin=<0 0 -10>, angularTarget=<0 0 1>}
      localJointOffset=<3.5 0 0>
     }]
   }
   n20 = Anim Dynamics - Bone: dyn_jacket_chainB_02_r!(ComponentPose=@n19, ExternalForce="0,0,0") -> (Pose) {
    angularDampingOverride=0.999
    angularSpringConstant=100
    bAngularSpring=true
    bGravityOverrideInSimSpace=true
    bOverrideAngularDamping=true
    numSolverIterationsPostUpdate=4
    numSolverIterationsPreUpdate=16
    simulationSpace=RootRelative
    alphaBoolBlend {blendOption=Linear}
    boundBone {boneName=dyn_jacket_chainB_02_r}
    chainEnd {boneName=dyn_jacket_chainB_02_r}
    componentLinearAccScale=<1 1 1>
    componentLinearVelScale=<1 1 1>
    physicsBodyDefinitions [{
      sphereCollisionRadius=0
      boundBone {boneName=dyn_jacket_chainB_02_r}
      boxExtents=<12 20 25>
      constraintSetup {angularLimitsMax=<0 30 10>, angularLimitsMin=<0 0 -10>, angularTarget=<0 0 1>}
      localJointOffset=<3.5 0 0>
     }]
   }
   n21 = Anim Dynamics - Bone: dyn_jacket_chainC_02_r!(ComponentPose=@n20, ExternalForce="0,0,0") -> (Pose) {
    angularDampingOverride=0.999
    angularSpringConstant=100
    bAngularSpring=true
    bGravityOverrideInSimSpace=true
    bOverrideAngularDamping=true
    numSolverIterationsPostUpdate=4
    numSolverIterationsPreUpdate=16
    simulationSpace=RootRelative
    alphaBoolBlend {blendOption=Linear}
    boundBone {boneName=dyn_jacket_chainC_02_r}
    chainEnd {boneName=dyn_jacket_chainC_02_r}
    componentLinearAccScale=<1 1 1>
    componentLinearVelScale=<1 1 1>
    physicsBodyDefinitions [{
      sphereCollisionRadius=0
      boundBone {boneName=dyn_jacket_chainC_02_r}
      boxExtents=<12 20 25>
      constraintSetup {angularLimitsMax=<0 30 10>, angularLimitsMin=<0 0 -10>, angularTarget=<0 0 1>}
      localJointOffset=<3.5 0 0>
     }]
   }
   n22 = Flap Node - Bone: dyn_scarf_offset!(ComponentPose:Component Space Pose Link Structure=@n21, Alpha:float=1) -> (Pose:Component Space Pose Link Structure) {
    is=title FortAnimGraphNode_Flap
    bUseClamp=true
    targetClampMax=3.14
    targetComponent=RotationZ
    alphaBoolBlend {blendOption=Linear}
    sourceDrives [{
      bUseRange=false
      drivingCurve=null
      multiplier=1
      rangeMax=0
      rangeMin=0
      remappedMax=0
      remappedMin=0
      sourceComponent=RotationZ
      useQuaternion=true
      sourceBone {boneName=head}
     }]
    targetBone {boneName=dyn_scarf_offset}
   }
   Output Pose(Result=Component To Local(ComponentPose=@n22)) {layerGroup=None}
   Save cached pose 'Cache_01'() {cacheName=Cache_01}
   "integer >= integer"!(A:int=Get CurrentLOD(), B:int=2) -> (ReturnValue:bool) {is=title K2Node_CallFunction}
  ]}
 EventGraph {nodes [
   Event BlueprintUpdateAnimation()
   > n1 = Branch(Condition=Get isWearingBackbling())
   > n2 = Get Dynamic Material Instance(SkeletalMeshComponent=GetPartSkeletalMeshComponent())
   > n3 = SetScalarParameterValue(self=@n2, ParameterName=Hide Element 01, Value=1)
   Do Once(execute=[@n3.then, @n1.else])
   > n4 = Cast To FortPlayerPawn(Object=TryGetPawnOwner())
   > Is Valid(InputObject=GetSkeletalMeshForPartType(self=@n4, PartType=Backpack))
   > Set isWearingBackbling(isWearingBackbling=true)
   TryGetPawnOwner()
  ]}
]
variables [isWearingBackbling {type=boolean}]
