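# Physics asset for the medium male Scarecrow player body, named as the collision setup for its cloth.
# It has bodies for the left and right thigh, calf and foot, but only the two calves carry collision shapes, one capsule each.
# Constraints link each foot to its calf and each calf to its thigh, with all swing and twist motion left free.
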PhysicsAsset /BRCosmetics/Characters/Player/Male/Medium/Bodies/M_MED_Scarecrow/Meshes/M_MED_Scarecrow_Body_Cloth_Physics
bodies {
 calf_l [{shapeType=Capsule, length=30.8668, radius=7.8569, center=<-20.7429 -1.3648 -0.5538>, rotation=<rot 0 90.7572 -89.0068>}]
 calf_r [{shapeType=Capsule, length=30.9928, radius=7.803, center=<20.3739 1.4027 -0.6106>, rotation=<rot 0 -89.5678 -87.3833>}]
 foot_l []
 foot_r []
 thigh_l []
 thigh_r []
}
constraints [
 {
  bone1Name=calf_l
  bone2Name=thigh_l
  swing1LimitDegrees=45
  swing1Motion=Free
  swing2LimitDegrees=45
  swing2Motion=Free
  twistLimitDegrees=45
  twistMotion=Free
 }
 {
  bone1Name=calf_r
  bone2Name=thigh_r
  swing1LimitDegrees=45
  swing1Motion=Free
  swing2LimitDegrees=45
  swing2Motion=Free
  twistLimitDegrees=45
  twistMotion=Free
 }
 {
  bone1Name=foot_r
  bone2Name=calf_r
  swing1LimitDegrees=45
  swing1Motion=Free
  swing2LimitDegrees=45
  swing2Motion=Free
  twistLimitDegrees=45
  twistMotion=Free
 }
 {
  bone1Name=foot_l
  bone2Name=calf_l
  swing1LimitDegrees=45
  swing1Motion=Free
  swing2LimitDegrees=45
  swing2Motion=Free
  twistLimitDegrees=45
  twistMotion=Free
 }
]
