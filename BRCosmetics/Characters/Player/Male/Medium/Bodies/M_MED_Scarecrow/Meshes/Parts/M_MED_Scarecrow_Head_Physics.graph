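# Physics asset for the Scarecrow male medium player cosmetic's head mesh, simulating the two hat flaps.
# Capsule bodies cover the left and right dyn_hat_flap bones, with a small zero-length capsule on the head bone as their anchor.
# Each flap is constrained to the head with angular swing limits, tighter on swing 1 than swing 2, and twist locked.
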PhysicsAsset /BRCosmetics/Characters/Player/Male/Medium/Bodies/M_MED_Scarecrow/Meshes/Parts/M_MED_Scarecrow_Head_Physics
bodies {
 dyn_hat_flap_l [{shapeType=Capsule, length=9, radius=6, center=<6 0 0>, rotation=<rot 90 0 0>}]
 dyn_hat_flap_r [{shapeType=Capsule, length=9, radius=6, center=<6 0 0>, rotation=<rot 90 0 0>}]
 head [{shapeType=Capsule, length=0, radius=3}]
}
constraints [
 {bone1Name=dyn_hat_flap_r, bone2Name=head, swing1LimitDegrees=10, swing2LimitDegrees=45, twistLimitDegrees=45, twistMotion=Locked}
 {bone1Name=dyn_hat_flap_l, bone2Name=head, swing1LimitDegrees=10, swing2LimitDegrees=45, twistLimitDegrees=45, twistMotion=Locked}
]
